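# HLOD material for the Nexus Zero Point core, drawing a glowing orb with a dark default base color.
# Two Fresnel layers blend inner and outer colors, with a static switch for a simple two-color lerp.
# Layered sines flicker the glow, with MPC_Nexus Destabilize driving speed, amplitude and inner tint.
# An optional world position offset spin rotates around the local Z axis, and MPC_RadishPortal supplies color and intensity multipliers.
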Material /BlastBerryMapContent/Game/Blueprints/ZP/TechArt/NexusCore/M_Radish_Nex_Core_HLOD
graph {
 properties {
  bUsedWithMeshParticles=true
  bUsedWithParticleSprites=true
  numCustomizedUVs=3
  pixelDepthOffsetMode=PDOM_Legacy
  usedShadingModels=Default Lit
 }
 nodes [
  n1 = Normalize(VectorInput=Transform(None=Constant3Vector() {constant=<rgba 0 0 1 0>}) {transformSourceType=TRANSFORMSOURCE_Local})
  n2 = Multiply(A=Time(), B=ScalarParameter() {defaultValue=0.1, parameterName=SpinSpeed}) {constB=0.1}
  AppendVector(
   A=MaterialFunctionCall(Rotation Axis=@n1, Rotation Angle=@n2) {materialFunction=&FixRotateAboutAxisNormals}.New Tangent Space Vertex Normal B
   B=Constant()
  )
  CollectionParameter() {collection=/Game/Athena/Environments/Cattus/Materials/MPC_CattusPOI, parameterName=ZeroPointHidden}
  CollectionParameter() {collection=/BlastBerryMapContent/Game/Blueprints/ZP/TechArt/Portals/MPC_RadishPortal, parameterName=ZeroPoint Color}
  Constant()
  Constant() {r=1}
  MaterialFunctionCall() {materialFunction=&FadeBasedOnViewAngle}
  n3 = ScalarParameter() {defaultValue=0.8, parameterName=FlickerSpeed}
  n4 = Multiply(
   A=Time()
   B=LinearInterpolate(
    A=@n3
    B=Multiply(A=@n3) {constB=2}
    Alpha=CollectionParameter() {collection=/Game/Athena/Environments/Nexus/Materials/FXshader/MPC_Nexus, parameterName=Destabilize}
   )
  )
  n5 = Add(A=Multiply(B=Sine(None=@n4) {period=0.55}), B=Multiply(A=Sine(None=@n4) {period=0.11}) {constB=0.2})
  n6 = Multiply(A=Sine(None=@n4) {period=0.175}) {constB=0.3}
  n7 = Multiply(A=Sine(None=@n4) {period=3.25}, B=Sine(None=@n4) {period=2.325})
  n8 = CollectionParameter() {collection=&MPC_Nexus, parameterName=Destabilize}
  n9 = Multiply(A=Multiply(A=Add(A=@n5, B=@n6), B=ConstantBiasScale(None=@n7)), B=LinearInterpolate(Alpha=@n8) {constA=0.3})
  n10 = Fresnel(ExponentIn=Add(A=ScalarParameter() {defaultValue=2, parameterName=Fresnel1}, B=Multiply(A=@n9) {constB=0.03})) {baseReflectFraction=0}
  n11 = VectorParameter() {parameterName=Outer, defaultValue=<rgba 0.2078 0.191 0.4844 1>}
  n12 = LinearInterpolate(Alpha=@n9) {constA=0.4, constB=1.1}
  n13 = VectorParameter() {parameterName=Inner, defaultValue=<rgba 6.6855 15.8898 25 1>}
  n14 = VectorParameter() {parameterName=DestabilizeTint, defaultValue=<rgba 4 0.8631 2.934 0>}
  n15 = Multiply(A=@n9) {constB=0.05}
  n16 = Fresnel(ExponentIn=Add(A=@n15, B=ScalarParameter() {defaultValue=0.14, parameterName=Fresnel2})) {baseReflectFraction=0}
  n17 = ScalarParameter() {defaultValue=4, parameterName=Fresnel2Power}
  n18 = Multiply(A=LinearInterpolate(A=@n13.RGB, B=Multiply(A=@n13.RGB, B=@n14.RGB), Alpha=@n8), B=Power(Base=OneMinus(None=@n16), Exp=Add(A=@n15, B=@n17)))
  n19 = StaticSwitchParameter(
   True=LinearInterpolate(
    A=VectorParameter() {parameterName=InsiderColor, defaultValue=<rgba 0 0 0 1>}.RGB
    B=VectorParameter() {parameterName=OutsiderColor, defaultValue=<rgba 1 1 1 1>}.RGB
    Alpha=@n10
   )
   False=Add(A=Multiply(A=@n11.RGB, B=@n10), B=Multiply(A=@n12, B=@n18))
  ) {parameterName=Use_lerp_Color}
  Multiply(
   A=CollectionParameter() {collection=&MPC_RadishPortal, parameterName=ZP_Emissive Intensity}
   B=@n19
  )
  Multiply(
   A=@n19
   B=ComponentMask(None=CollectionParameter() {collection=&MPC_RadishPortal, parameterName=ZeroPoint Color}) {b=true, g=true, r=true}
  )
  ScalarParameter() {parameterName=Roughness}
  n20 = Multiply(B=AppendVector(A=Constant2Vector(), B=ScalarParameter() {defaultValue=-5000, parameterName=Param_2}))
  n21 = MaterialFunctionCall() {materialFunction=&ObjectPivotPoint}
  n22 = WorldPosition()
  StaticSwitchParameter(
   True=Add(A=@n20, B=RotateAboutAxis(NormalizedRotationAxis=@n1, RotationAngle=@n2, PivotPoint=@n21.Object Pivot Location, Position=@n22.XYZ))
   False=@n20
  ) {parameterName=UseWOPSpin}
  n23 = TextureCoordinate() {coordinateIndex=2}
  StaticSwitchParameter(
   True=AppendVector(A=TextureCoordinate() {coordinateIndex=1}, B=ComponentMask(None=@n23) {r=true})
   False=Constant3Vector() {constant=<rgba 0 0 1 0>}
  ) {parameterName=UseWOPSpin}
  n24 = TextureCoordinate() {coordinateIndex=2}
  StaticSwitchParameter(
   True=AppendVector(A=TextureCoordinate() {coordinateIndex=1}, B=ComponentMask(None=@n24) {r=true})
   False=Constant3Vector() {constant=<rgba 0 0 1 0>}
  ) {parameterName=UseWOPSpin}
  VectorParameter() {parameterName=BaseColor, defaultValue=<rgba 0 0 0 1>}
 ]
}
